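# Physics asset for the Dummeez pickaxe, a two-handed melee cosmetic with a jointed spine and head.
# Single capsule bodies cover the handle bone, four spine bones and the head.
# Constraints link them in one chain from handle through spine_01 to spine_04 to head.
# Tight swing and twist limits keep the spine nearly rigid, while the head joint is loosest, allowing the most wobble at the tip.
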PhysicsAsset /BRCosmetics/Weapons/FORT_Melee/Pickaxe_Dummeez/Meshes/Dummeez_Axe_Physics
bodies {
 Melee_TwoHanded_Handle [{shapeType=Capsule, length=55.1489, radius=4.6515, center=<-0.1975 -0.1484 26.2152>, rotation=<rot 0.5725 -90 -1.1887>}]
 head [{shapeType=Capsule, length=2.25, radius=6.25, center=<-2.5 0 0>, rotation=<rot 83.5727 180 180>}]
 spine_01 [{shapeType=Capsule, length=5.5, radius=7.25, center=<-0.15 0 0>, rotation=<rot 81.3001 0 0>}]
 spine_02 [{shapeType=Capsule, length=7.75, radius=6.25, center=<-0.15 -0.0001 0>, rotation=<rot 81.8307 0 0>}]
 spine_03 [{shapeType=Capsule, length=8, radius=5.5, center=<-0.2 0 0>, rotation=<rot 87.5431 0 0>}]
 spine_04 [{shapeType=Capsule, length=3, radius=6, center=<-0.2 0 0>, rotation=<rot 87.844 0 0>}]
}
constraints [
 {bone1Name=spine_01, bone2Name=Melee_TwoHanded_Handle, swing1LimitDegrees=5, swing2LimitDegrees=5, twistLimitDegrees=3}
 {bone1Name=spine_02, bone2Name=spine_01, swing1LimitDegrees=2, swing2LimitDegrees=2, twistLimitDegrees=3}
 {bone1Name=spine_03, bone2Name=spine_02, swing1LimitDegrees=4, swing2LimitDegrees=4, twistLimitDegrees=3}
 {bone1Name=spine_04, bone2Name=spine_03, swing1LimitDegrees=6, swing2LimitDegrees=6, twistLimitDegrees=3}
 {bone1Name=head, bone2Name=spine_04, swing1LimitDegrees=16, swing2LimitDegrees=16, twistLimitDegrees=5}
]
